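; Gameplay ability for the Heavy specialist NPC that spawns two surprise projectiles when the bot dies.
; On activation it caches the bot's pawn and AI controller, then binds conversion and death events.
; Convert and unconvert events update the projectiles' instigator to the converting pawn or the unconverted bot.
; Unless blocking tags are present, death spawns both projectiles above the body after separate scalable delays, carrying an explosion effect spec.
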(event Ability|EventActivateAbility
  (bind _returnvalue (Ability|GetAvatarActorfromActorInfo))
  (Ability|CommitAbility)
  (bind _asfort_player_pawn (Utilities|Casting|CastToFortPlayerPawn _returnvalue)
    (:then
      (Variables|InternalData|SetBotFortPlayerPawn _asfort_player_pawn)
      (bind _asfort_athena_aibot_controller (Utilities|Casting|CastToFortAthenaAIBotController (AI|GetAIController _returnvalue))
        (:then
          (Variables|InternalData|SetBotController _asfort_athena_aibot_controller)
          (ConvertEvents|AssignOnConvertedEvent (Variables|/Script/FortniteGame.FortPlayerPawn|GetConvertComponent (Variables|InternalData|GetBotFortPlayerPawn)) (delegate Custom|OnConvertedEvent_Event))
          (Death|BindEventtoOnAgentDiedEvent (Variables|InternalData|GetBotController) (delegate Custom|BotDied)))
        (:CastFailed
          (Ability|EndAbility))))
    (:CastFailed
      (Ability|EndAbility))))

(event Custom|BotDied
  (params (BotController :type Fort%20Athena%20AIBot%20Controller%20Object%20Reference) (AgentPawn :type Fort%20Pawn%20Object%20Reference))
  (Variables|InternalData|SetBotLocation (Transformation|GetActorLocation AgentPawn))
  (if (not (GameplayTags|HasAnyMatchingGameplayTags (Variables|InternalData|GetBotFortPlayerPawn) (Variables|Default|GetTG_PreventAbility)))
    (Variables|InternalData|SetSpecOnExplode (Ability|MakeGameplayEffectContainerSpec (Variables|InternalData|GetApplicationTag00)))
    (Utilities|FlowControl|Sequence
      (:then_0
        (Utilities|FlowControl|Delay (ScalableFloat|GetValueatLevel (Variables|Default|GetSpawnDelayFirstProjectile)))
        (Projectile|SpawnProjectile (Variables|Default|GetProjectileToSpawn) (Variables|InternalData|GetInstigator%20Pawn) (+ (Variables|InternalData|GetBotLocation) "0.000000,50.000000,140.000000") "0, 0, 0" (Math|Conversions|RotationFromXVector (+ (Variables|InternalData|GetBotLocation) "0.000000,50.000000,140.000000")) (Variables|InternalData|GetSpecOnExplode) (Variables|InternalData|GetSpecOnExplode)))
      (:then_1
        (Utilities|FlowControl|Delay (ScalableFloat|GetValueatLevel (Variables|Default|GetSpawnDelaySecondProjectile)))
        (Projectile|SpawnProjectile (Variables|Default|GetProjectileToSpawn) (Variables|InternalData|GetInstigator%20Pawn) (+ (Variables|InternalData|GetBotLocation) "50.000000,0.000000,140.000000") "0, 0, 0" (Math|Conversions|RotationFromXVector (+ (Variables|InternalData|GetBotLocation) "50.000000,0.000000,140.000000")) (Variables|InternalData|GetSpecOnExplode) (Variables|InternalData|GetSpecOnExplode))))))

(event Custom|OnConvertedEvent_Event
  (params (InstigatorPawn :type Fort%20Pawn%20Object%20Reference) (ConvertedPawn :type Fort%20Pawn%20Object%20Reference))
  (Variables|InternalData|SetInstigator%20Pawn InstigatorPawn)
  (ConvertEvents|BindEventtoOnUnconvertedEvent (Variables|/Script/FortniteGame.FortPlayerPawn|GetConvertComponent (Variables|InternalData|GetBotFortPlayerPawn)) (delegate Custom|OnUnconvertedEvent_Event)))

(event Custom|OnUnconvertedEvent_Event
  (params (UnconvertedPawn :type Fort%20Pawn%20Object%20Reference) (UnconvertReason :type EUnconvertReason%20Enum :default "StealByOtherPlayer"))
  (switch Utilities|FlowControl|Switch|SwitchonEUnconvertReason UnconvertReason
    (:StealByOtherPlayer
      _)
    (:TooManyConverted
      _)
    (:TooFarFromConverter
      _)
    (:GameplayTagBlockConvert
      _)
    (:EndPlay
      _)
    (:Dismiss
      _)
    (:Unknown
      _))
  (Variables|InternalData|SetInstigator%20Pawn UnconvertedPawn))

(event Ability|EventFailedToActivatePassiveAbility
  (params (ActorInfo :type Gameplay%20Ability%20Actor%20Info%20Structure)))

(event Ability|EventOnEndAbility
  (params (bWasCancelled :type Boolean)))
